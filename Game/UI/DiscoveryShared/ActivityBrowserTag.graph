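# Activity browser tag widget for Discovery UI, showing a label with an optional icon on a material-backed background.
# OnTagUpdated switches on tag type to pick the icon and background material, while XP swaps the whole tag for an XP image.
# Text and Epic icon tint come from the background dynamic material's Pressed_Fill parameter, keeping colors matched to the material.
# PreConstruct applies default styling and sets a taller height on mobile.
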WidgetBlueprint /Game/UI/DiscoveryShared/ActivityBrowserTag
graphs [
 EventGraph {nodes [
   n1 = UpdateTagStyle!() -> (OutputDelegate, then, TagType:EFortActivityBrowserTagType Enum=Default) {is=title K2Node_CustomEvent}
   > n2 = Switch on EFortActivityBrowserTagType!(execute, Selection:EFortActivityBrowserTagType Enum=@n1.TagType) -> (Default:Exec, EpicOriginal:Exec, SquadSize:Exec, MaxPlayers:Exec, XP:Exec, PurchaseRequired:Exec, DownloadRequired:Exec, LinkCategory:Exec) {is=title K2Node_SwitchEnum}
   > SetActiveWidget(
    self=Get IconSwitcher() -> (IconSwitcher:Common Visibility Switcher Object Reference)
    Widget=Get Spacer_NoIcon() -> (Spacer_NoIcon:Spacer Object Reference)
   )
   > n3 = SetDefaultStyling()
   SetTextColorFromMaterialInstance(execute=[@n8.then, @n6.then, @n4.then, @n9.then, @n5.then, @n3.then, @n7.then])
   > SetActiveWidgetIndex(self:Common Visibility Switcher Object Reference=Get TagSwitcher() -> (TagSwitcher:Common Visibility Switcher Object Reference), Index=0)
   SetActiveWidget(
    execute=@n2.DownloadRequired
    self=Get IconSwitcher()
    Widget=Get Image_DownloadRequiredIcon() -> (Image_DownloadRequiredIcon:Image Object Reference)
   )
   > n4 = SetTagMaterial!(execute, self, NewMaterial:Object Reference=Get TagMaterial_DownloadRequired() -> (TagMaterial_DownloadRequired:Object Reference)) -> (then)
   SetActiveWidget(execute=@n2.SquadSize, self=Get IconSwitcher(), Widget=Get Image_PlayerIcon() -> (Image_PlayerIcon:Image Object Reference))
   > n5 = SetDefaultStyling()
   SetActiveWidget(execute=@n2.EpicOriginal, self=Get IconSwitcher(), Widget=Get Image_EpicContentIcon() -> (Image_EpicContentIcon:Image Object Reference))
   > n6 = SetTagMaterial(NewMaterial=Get TagMaterial_EpicTag() -> (TagMaterial_EpicTag:Object Reference))
   SetActiveWidget(
    execute=@n2.PurchaseRequired
    self=Get IconSwitcher()
    Widget=Get Image_PurchaseRequiredIcon() -> (Image_PurchaseRequiredIcon:Image Object Reference)
   )
   > n7 = SetTagMaterial(NewMaterial=Get TagMaterial_PurchaseRequired() -> (TagMaterial_PurchaseRequired:Object Reference))
   SetActiveWidget(execute=@n2.LinkCategory, self=Get IconSwitcher(), Widget=Get Spacer_NoIcon())
   > n8 = SetDefaultStyling()
   SetActiveWidget(execute=@n2.MaxPlayers, self=Get IconSwitcher(), Widget=Get Spacer_NoIcon())
   > n9 = SetDefaultStyling()
   SetActiveWidgetIndex(execute=@n2.XP, self=Get TagSwitcher(), Index=1)
   n10 = Event PreConstruct()
   > SetDefaultStyling()
   > n11 = Sequence()
   > Branch(Condition=@n10.IsDesignTime)
   > UpdateTagStyle!(execute, self, TagType:EFortActivityBrowserTagType Enum=Default) -> (then) {is=title K2Node_CallFunction}
   SetTextColorFromMaterialInstance(execute=@n11.then_1)
   > SetHeightOverride(self=Get HeightSB() -> (HeightSB:Size Box Object Reference), InHeightOverride=SelectFloat(A=37, B=32, bPickA=IsMobileGame()))
   Event OnTagUpdated() {is=title K2Node_Event}
   > UpdateTagStyle(TagType=GetTagType!(self:Fort Activity Browser Tag Object Reference) -> (ReturnValue:EFortActivityBrowserTagType Enum=Default))
  ]}
 SetDefaultStyling {nodes [
   SetDefaultStyling() {is=title K2Node_FunctionEntry}
   > SetTagMaterial(NewMaterial=Get TagMaterial_Default() -> (TagMaterial_Default:Object Reference))
  ]}
 SetTagMaterial {nodes [
   n1 = SetTagMaterial!() -> (then, NewMaterial:Object Reference) {is=title K2Node_FunctionEntry}
   n2 = Get TagContainer() -> (TagContainer:Common Border Object Reference)
   n3 = Break Slate Brush!(SlateBrush:"Slate Brush Structure (by ref)"=Get Background!(self:Border Object Reference=@n2) -> (Background:Slate Brush Structure)) -> (
    TintColor:Slate Color Structure
    DrawAs:ESlateBrushDrawType Enum
    Tiling:ESlateBrushTileType Enum
    Mirroring:ESlateBrushMirrorType Enum
    ImageSize:Vector 2D Structure
    Margin:Margin Structure
    ResourceObject:Object Reference
    OutlineSettings:Slate Brush Outline Settings Structure
   ) {is=title K2Node_BreakStruct}
   > SetBrush(
    self:Border Object Reference=@n2
    InBrush=Make Slate Brush!(
     TintColor:Slate Color Structure=@n3
     DrawAs:ESlateBrushDrawType Enum=@n3.DrawAs
     Tiling:ESlateBrushTileType Enum=@n3.Tiling
     Mirroring:ESlateBrushMirrorType Enum=@n3.Mirroring
     ImageSize:Vector 2D Structure=@n3.ImageSize
     Margin:Margin Structure=@n3.Margin
     ResourceObject:Object Reference=@n1
     OutlineSettings:Slate Brush Outline Settings Structure=@n3.OutlineSettings
    ) -> (SlateBrush:Slate Brush Structure) {is=title K2Node_MakeStruct}
   )
  ]}
 SetTextColorFromMaterialInstance {nodes [
   SetTextColorFromMaterialInstance() {is=title K2Node_FunctionEntry}
   > n1 = GetDynamicMaterial(self:Border Object Reference=Get TagContainer())
   > n2 = Get Vector Parameter Value!(execute, self:Material Instance Dynamic Object Reference=@n1, ParameterName:Name=Pressed_Fill) -> (then, ReturnValue:Linear Color Structure) {is=title K2Node_CallFunction}
   n3 = Make Slate Color(SpecifiedColor=@n2)
   > SetColorAndOpacity(self=Get Text_TagDescription() -> (Text_TagDescription:Common Text Object Reference), InColorAndOpacity=@n3)
   > SetBrushTintColor!(execute, self:Image Object Reference=Get Image_EpicContentIcon(), TintColor:Slate Color Structure=@n3) -> (then)
  ]}
]
variables [
 TagMaterial_Default {type=Object, default=/Game/UI/DiscoveryShared/Art/MI_UI_Discovery_BasicTag}
 TagMaterial_DownloadRequired {type=Object, default=/Game/UI/DiscoveryShared/Art/MI_UI_Discovery_PurchaseRequiredTag}
 TagMaterial_EpicTag {type=Object, default=/Game/UI/DiscoveryShared/Art/MI_UI_Discovery_EpicTag}
 TagMaterial_PurchaseRequired {type=Object, default=&MI_UI_Discovery_PurchaseRequiredTag}
]
widgets {
 info {inheritedWidgetCount=1, parentClass=/Script/FortniteUI.FortActivityBrowserTag, rootWidgetClass=&SizeBox, widgetCount=13}
 widgets [
  HeightSB {
   widgetClassPath=&SizeBox
   bIsVariable=true
   widget=HeightSB
   properties {bOverride_MinDesiredHeight=true, heightOverride=32, minDesiredHeight=32}
  }
  TagSwitcher {widgetClassPath=&CommonVisibilitySwitcher, bIsVariable=true, parent=HeightSB, widget=TagSwitcher, slot {class=SizeBoxSlot}}
  TagContainer {
   widgetClassPath=&CommonBorder
   bIsVariable=true
   parent=TagSwitcher
   widget=TagContainer
   properties {
    background {
     bIsDynamicallyLoaded=false
     drawAs=Box
     imageType=NoImage
     mirroring=NoMirror
     resourceName=None
     resourceObject=&MI_UI_Discovery_BasicTag
     tiling=NoTile
     imageSize=<32 32>
     margin=<ltrb 0.5 0 0.5 0>
     outlineSettings {
      bUseBrushTransparency=false
      roundingType=HalfHeightRadius
      width=0
      color {colorUseRule=UseColor_Specified, specifiedColor=<rgba 0 0 0 0>}
      cornerRadii=<0 0 0 0>
     }
     tintColor {colorUseRule=UseColor_Specified, specifiedColor=<rgba 1 1 1 1>}
     uVRegion {bIsValid=false, max=<0 0>, min=<0 0>}
    }
    padding=<ltrb 18 0 18 0>
   }
   slot {class=CommonVisibilitySwitcherSlot}
  }
  HorizontalBox_0 {
   widgetClassPath=&HorizontalBox
   parent=TagContainer
   widget=HorizontalBox_0
   slot {class=BorderSlot, properties {padding=<ltrb 18 0 18 0>}}
  }
  IconSwitcher {
   widgetClassPath=&CommonVisibilitySwitcher
   bIsVariable=true
   parent=HorizontalBox_0
   widget=IconSwitcher
   slot {class=HorizontalBoxSlot}
  }
  Spacer_NoIcon {
   widgetClassPath=&Spacer
   bIsVariable=true
   parent=IconSwitcher
   widget=Spacer_NoIcon
   properties {size=<0 0>}
   slot {class=CommonVisibilitySwitcherSlot}
  }
  Image_PurchaseRequiredIcon {
   widgetClassPath=&Image
   bIsVariable=true
   parent=IconSwitcher
   widget=Image_PurchaseRequiredIcon
   properties {
    visibility=HitTestInvisible
    brush {
     bIsDynamicallyLoaded=false
     drawAs=Image
     imageType=NoImage
     mirroring=NoMirror
     resourceName=None
     resourceObject=/Game/UI/Foundation/Textures/Icons/Locks/T-Icon-Lock-32
     tiling=NoTile
     imageSize=<23 23>
     margin=<ltrb 0 0 0 0>
     outlineSettings {
      bUseBrushTransparency=false
      roundingType=HalfHeightRadius
      width=0
      color {colorUseRule=UseColor_Specified, specifiedColor=<rgba 0 0 0 0>}
      cornerRadii=<0 0 0 0>
     }
     tintColor {colorUseRule=UseColor_Specified, specifiedColor=<rgba 1 1 1 1>}
     uVRegion {bIsValid=false, max=<0 0>, min=<0 0>}
    }
    renderTransform {angle=-10, scale=<1 1>, shear=<0 0>, translation=<0 0>}
   }
   slot {class=CommonVisibilitySwitcherSlot, properties {verticalAlignment=VAlign_Center, padding=<ltrb -4 0 6 0>}}
  }
  Image_EpicContentIcon {
   widgetClassPath=&Image
   bIsVariable=true
   parent=IconSwitcher
   widget=Image_EpicContentIcon
   properties {
    visibility=HitTestInvisible
    brush {
     bIsDynamicallyLoaded=false
     drawAs=Image
     imageType=NoImage
     mirroring=NoMirror
     resourceName=None
     resourceObject=/Game/UI/DiscoveryShared/Art/M_UI_Discovery_EpicContentIcon
     tiling=NoTile
     imageSize=<28 28>
     margin=<ltrb 0 0 0 0>
     outlineSettings {
      bUseBrushTransparency=false
      roundingType=HalfHeightRadius
      width=0
      color {colorUseRule=UseColor_Specified, specifiedColor=<rgba 0 0 0 0>}
      cornerRadii=<0 0 0 0>
     }
     tintColor {colorUseRule=UseColor_Specified, specifiedColor=<rgba 0.2159 0.6584 1 1>}
     uVRegion {bIsValid=false, max=<0 0>, min=<0 0>}
    }
   }
   slot {class=CommonVisibilitySwitcherSlot, properties {verticalAlignment=VAlign_Center, padding=<ltrb -4 0 6 0>}}
  }
  Image_PlayerIcon {
   widgetClassPath=&Image
   bIsVariable=true
   parent=IconSwitcher
   widget=Image_PlayerIcon
   properties {
    visibility=Collapsed
    brush {
     bIsDynamicallyLoaded=false
     drawAs=Image
     imageType=NoImage
     mirroring=NoMirror
     resourceName=None
     resourceObject=/Game/UI/DiscoveryShared/Art/T_UI_Discovery_PlayerIcon
     tiling=NoTile
     imageSize=<20 20>
     margin=<ltrb 0 0 0 0>
     outlineSettings {
      bUseBrushTransparency=false
      roundingType=HalfHeightRadius
      width=0
      color {colorUseRule=UseColor_Specified, specifiedColor=<rgba 0 0 0 0>}
      cornerRadii=<0 0 0 0>
     }
     tintColor {colorUseRule=UseColor_Specified, specifiedColor=<rgba 1 1 1 1>}
     uVRegion {bIsValid=false, max=<0 0>, min=<0 0>}
    }
    colorAndOpacity=<rgba 0.2159 0.6584 1 1>
   }
   slot {class=CommonVisibilitySwitcherSlot, properties {verticalAlignment=VAlign_Center, padding=<ltrb 0 0 4 0>}}
  }
  Image_DownloadRequiredIcon {
   widgetClassPath=&Image
   bIsVariable=true
   parent=IconSwitcher
   widget=Image_DownloadRequiredIcon
   properties {brush {
     bIsDynamicallyLoaded=false
     drawAs=Image
     imageType=NoImage
     mirroring=NoMirror
     resourceName=None
     resourceObject=/Game/UI/DiscoveryShared/Art/T_UI_icon_contentdownload_32
     tiling=NoTile
     imageSize=<26 26>
     margin=<ltrb 0 0 0 0>
     outlineSettings {
      bUseBrushTransparency=false
      roundingType=HalfHeightRadius
      width=0
      color {colorUseRule=UseColor_Specified, specifiedColor=<rgba 0 0 0 0>}
      cornerRadii=<0 0 0 0>
     }
     tintColor {colorUseRule=UseColor_Specified, specifiedColor=<rgba 1 1 1 1>}
     uVRegion {bIsValid=false, max=<0 0>, min=<0 0>}
    }}
   slot {class=CommonVisibilitySwitcherSlot, properties {verticalAlignment=VAlign_Center, padding=<ltrb -4 0 6 0>}}
  }
  Text_TagDescription {
   widgetClassPath=&CommonTextBlock
   bInherited=true
   bIsVariable=true
   parent=HorizontalBox_0
   widget=Text_TagDescription
   properties {
    mobileFontSizeMultiplier=1.2
    text=Tag Name
    textTransformPolicy=ToUpper
    visibility=HitTestInvisible
    colorAndOpacity {colorUseRule=UseColor_Specified, specifiedColor=<rgba 0.2159 0.6584 1 1>}
    font {
     bForceMonospaced=false
     bMaterialIsStencil=false
     fontMaterial=null
     fontObject=&BurbankSmall
     letterSpacing=0
     monospacedWidth=1
     size=15
     skewAmount=0
     typefaceFontName=Black
     outlineSettings {
      bApplyOutlineToDropShadows=false
      bMiteredCorners=false
      bSeparateFillAlpha=false
      outlineMaterial=null
      outlineSize=0
      outlineColor=<rgba 0 0 0 1>
     }
    }
    margin=<ltrb 0 -4.5 0 -5>
    renderTransform {angle=0, scale=<1 1>, shear=<-12 0>, translation=<0 0>}
    shadowOffset=<0 0>
   }
   slot {class=HorizontalBoxSlot, properties {verticalAlignment=VAlign_Center}}
  }
  Image_XP {
   widgetClassPath=&Image
   bIsVariable=true
   parent=TagSwitcher
   widget=Image_XP
   properties {
    brush {
     bIsDynamicallyLoaded=false
     drawAs=Image
     imageType=FullColor
     mirroring=NoMirror
     resourceName=None
     resourceObject=/Game/UI/Foundation/Textures/Icons/Items/UniversalXP/T_UI_UniversalXP_Icon_Standard_64
     tiling=NoTile
     imageSize=<64 64>
     margin=<ltrb 0 0 0 0>
     outlineSettings {
      bUseBrushTransparency=false
      roundingType=HalfHeightRadius
      width=0
      color {colorUseRule=UseColor_Specified, specifiedColor=<rgba 0 0 0 0>}
      cornerRadii=<0 0 0 0>
     }
     tintColor {colorUseRule=UseColor_Specified, specifiedColor=<rgba 1 1 1 1>}
     uVRegion {bIsValid=false, max=<0 0>, min=<0 0>}
    }
    renderTransform {angle=0, scale=<1 2>, shear=<0 0>, translation=<0 0>}
   }
   slot {class=CommonVisibilitySwitcherSlot, properties {horizontalAlignment=HAlign_Left, verticalAlignment=VAlign_Center}}
  }
  SqueegeeInjectionSlot {
   widgetClassPath=/Game/UI/Plugin/Squeegee/SqueegeeInjectionSlot
   bIsVariable=true
   parent=TagSwitcher
   widget=SqueegeeInjectionSlot
   properties {styleParams {bShowDetails=true, fontSize=15, paddingSize=0.5, iconSize=<32 32>, iconSkewSize=<6 32>}}
   slot {class=CommonVisibilitySwitcherSlot}
  }
 ]
}
